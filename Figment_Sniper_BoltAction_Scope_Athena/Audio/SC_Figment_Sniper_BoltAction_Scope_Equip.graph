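# Sound cue for the equip sound of the Figment scoped bolt-action sniper rifle.
# It plays a single sniper cock wave through a Modulator at default settings, after a short fixed delay.
# The cue uses the weapon foley sound class and borrows the weapon reload attenuation, with reduced volume.
# Concurrency override and the debug flag are both enabled.
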SoundCue /Figment_Sniper_BoltAction_Scope_Athena/Audio/SC_Figment_Sniper_BoltAction_Scope_Equip
root=Delay_0
nodes [
 {id=Delay_0, children [Modulator_0], properties {delayMax=0.1, delayMin=0.1}}
 {id=Modulator_0, children [WavePlayer_1]}
 {id=WavePlayer_1, properties {soundWaveAssetPtr=&/rifle_sniper_single_cock}}
]
properties {
 attenuationSettings=&Fort_Weapon_Reload_Att
 bDebug=true
 bOverrideConcurrency=true
 maxDistance=2250
 soundClassObject=&Fort_WeaponFoley
 volumeMultiplier=0.702
}
