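# Sound cue for the start of the reload on the Figment scoped bolt-action sniper rifle.
# A single wave player feeds a rifle clip-out sound straight to the output at reduced volume, with no randomization or modulation.
# It uses the shared weapon reload attenuation and weapon foley sound class.
# It also overrides concurrency and has its debug flag left on.
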SoundCue /Figment_Sniper_BoltAction_Scope_Athena/Audio/SC_Figment_Sniper_BoltAction_Scope_ReloadStart
root=WavePlayer_1
nodes [{id=WavePlayer_1, properties {soundWaveAssetPtr=&/rifle_sniper_single_clip_out}}]
properties {
 attenuationSettings=&Fort_Weapon_Reload_Att
 bDebug=true
 bOverrideConcurrency=true
 maxDistance=2250
 soundClassObject=&Fort_WeaponFoley
 volumeMultiplier=0.4
}
